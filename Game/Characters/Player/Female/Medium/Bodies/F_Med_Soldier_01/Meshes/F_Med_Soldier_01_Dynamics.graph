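# Physics asset for the dynamic dog tags (dyn_tags bone) on the F_Med_Soldier_01 female player body.
# A small body on dyn_tags is joined by a single constraint to a large, offset body on spine_05.
# The constraint sets equal swing limit angles and a much tighter twist limit.
# Both bodies use zero-length capsules, so they act as spheres.
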PhysicsAsset /Game/Characters/Player/Female/Medium/Bodies/F_Med_Soldier_01/Meshes/F_Med_Soldier_01_Dynamics
bodies {
 dyn_tags [{shapeType=Capsule, length=0, radius=1.5, center=<0 0.9603 0>, rotation=<rot 0 0 -6.8521>}]
 spine_05 [{shapeType=Capsule, length=0, radius=41.8418, center=<-19.3893 -25.7659 0.6715>, rotation=<rot 3.4939 -89.9998 -1.2946>}]
}
constraints [{bone1Name=dyn_tags, bone2Name=spine_05, swing1LimitDegrees=45, swing2LimitDegrees=45, twistLimitDegrees=5}]
