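# Post-copy AnimBlueprint for the FuzzyBearIndependence Star female body, adding secondary motion to cosmetic bones.
# It copies the owner mesh's pose, then chains AnimDynamics on bandana and tags, a knee-pad constraint, leg RigidBody and pocket offsets.
# Tag simulation alpha starts ramping once customization loads; grenade and several duplicate physics nodes sit unconnected.
# Each update checks the pawn's JumpBoots gameplay tag to set the "Hide Element 03" material parameter.
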AnimBlueprint /BRCosmetics/Characters/Player/Female/Medium/Bodies/F_MED_FuzzyBearIndependence_Star/Meshes/F_MED_FuzzyBearIndependance_Star_AnimBP
target_skeleton=/Game/Characters/Player/Female/Medium/Bodies/F_Med_Soldier_01/Meshes/F_Med_Soldier_01_Skeleton
graphs [
 AnimGraph {nodes [
   n1 = Anim Dynamics - Bone: dyn_grenade_back(Alpha=1, ExternalForce="0,0,0") {
    angularSpringConstant=1000
    bAngularSpring=true
    bGravityOverrideInSimSpace=true
    lODThreshold=1
    numSolverIterationsPostUpdate=4
    numSolverIterationsPreUpdate=16
    alphaBoolBlend {blendOption=Linear}
    boundBone {boneName=dyn_grenade_back}
    physicsBodyDefinitions [{
      sphereCollisionRadius=0
      boundBone {boneName=dyn_grenade_back}
      boxExtents=<10 10 20>
      constraintSetup {
       angularTargetAxis=AxisY
       twistAxis=AxisZ
       angularLimitsMax=<20 0 0>
       angularLimitsMin=<0 -20 0>
       angularTarget=<0 0 0>
       linearAxesMax=<0 0 0.5>
       linearAxesMin=<0 0 -0.5>
      }
      localJointOffset=<0 0 -2>
     }]
    sphericalLimits [{limitRadius=16, drivingBone {boneName=pelvis}, sphereLocalOffset=<-1 1 0>}]
   }
   Anim Dynamics - Bone: dyn_grenade_front(ComponentPose=@n1, Alpha=1, ExternalForce="0,0,0") {
    angularSpringConstant=5000
    bGravityOverrideInSimSpace=true
    gravityScale=0.2
    lODThreshold=1
    numSolverIterationsPostUpdate=4
    numSolverIterationsPreUpdate=16
    alphaBoolBlend {blendOption=Linear}
    boundBone {boneName=dyn_grenade_front}
    physicsBodyDefinitions [{
      sphereCollisionRadius=0
      boundBone {boneName=dyn_grenade_front}
      boxExtents=<10 10 20>
      constraintSetup {
       angularTargetAxis=AxisY
       twistAxis=AxisZ
       angularLimitsMax=<20 0 0>
       angularLimitsMin=<0 -20 0>
       angularTarget=<0 0 0>
       linearAxesMax=<0 0 0.5>
       linearAxesMin=<0 0 -0.5>
      }
      localJointOffset=<0 0 -2>
     }]
    sphericalLimits [{limitRadius=16, drivingBone {boneName=pelvis}, sphereLocalOffset=<-1 1 0>}]
   }
   Anim Dynamics - Bone: dyn_tags(Alpha=1, ExternalForce="0,0,0") {
    angularBiasOverride=1
    angularDampingOverride=0.98
    bGravityOverrideInSimSpace=true
    lODThreshold=1
    numSolverIterationsPostUpdate=4
    numSolverIterationsPreUpdate=16
    alphaBoolBlend {blendOption=Linear}
    alphaScaleBiasClamp {bInterpResult=true, interpSpeedIncreasing=0.05, scale=0.7}
    boundBone {boneName=dyn_tags}
    physicsBodyDefinitions [{
      sphereCollisionRadius=0
      boundBone {boneName=dyn_tags}
      boxExtents=<5 2 10>
      constraintSetup {twistAxis=AxisZ, angularLimitsMax=<25 45 0>, angularLimitsMin=<0 -45 0>, angularTarget=<0 0 0>}
      localJointOffset=<0 1 -1>
     }]
    relativeSpaceBone {boneName=dyn_tags}
   }
   Anim Dynamics - Bone: dyn_arm_r_bandana(Alpha=.65, ExternalForce="0,0,0") {
    bGravityOverrideInSimSpace=true
    lODThreshold=1
    numSolverIterationsPostUpdate=4
    numSolverIterationsPreUpdate=16
    alphaBoolBlend {blendOption=Linear}
    boundBone {boneName=dyn_arm_r_bandana}
    physicsBodyDefinitions [{
      sphereCollisionRadius=0
      boundBone {boneName=dyn_arm_r_bandana}
      boxExtents=<20 20 10>
      constraintSetup {twistAxis=AxisY, angularLimitsMax=<45 0 45>, angularLimitsMin=<-45 0 -45>, angularTarget=<0 0 0>}
      localJointOffset=<0 -6 0>
     }]
   }
   RigidBody() {
    overridePhysicsAsset=/Game/Characters/Player/Female/Medium/Bodies/F_Med_Soldier_01/Meshes/F_Med_Soldier_01_Dynamics
    alphaBoolBlend {blendOption=Linear}
   }
   n2 = Get MeshToCopy() -> (MeshToCopy:Skeletal Mesh Component Object Reference)
   n3 = "Transform (Modify) Bone - Bone: dyn_tags"(
    ComponentPose=Local To Component(LocalPose=Copy Pose From Mesh(SourceMeshComponent=@n2))
    Alpha=1
    Translation="0,0,0"
    Rotation="0,-6.5,0"
    Scale="1,1,1"
   ) {
    rotationMode=BMM_Additive
    rotationSpace=BCS_ParentBoneSpace
    alphaBoolBlend {blendOption=Linear}
    boneToModify {boneName=dyn_tags}
    rotation {yaw=-6.5}
   }
   n4 = Anim Dynamics - Bone: dyn_arm_r_bandana(ComponentPose=@n3, Alpha=0.5, ExternalForce="0,0,0") {
    alpha=0.9
    angularSpringConstant=100
    bAngularSpring=true
    bGravityOverrideInSimSpace=true
    numSolverIterationsPostUpdate=4
    numSolverIterationsPreUpdate=16
    simulationSpace=RootRelative
    alphaBoolBlend {blendOption=Linear}
    boundBone {boneName=dyn_arm_r_bandana}
    componentAppliedLinearAccClamp=<30000 30000 30000>
    componentLinearAccScale=<0.4 0.3 0.3>
    componentLinearVelScale=<0.4 0.3 0.3>
    gravityOverride {z=-980}
    physicsBodyDefinitions [{
      sphereCollisionRadius=0
      boundBone {boneName=dyn_arm_r_bandana}
      boxExtents=<15 5 5>
      constraintSetup {angularTargetAxis=AxisZ, twistAxis=AxisY, angularLimitsMax=<15 0 15>, angularLimitsMin=<-15 0 -15>, angularTarget=<0 0 1>}
      localJointOffset=<0 0 0>
     }]
   }
   n5 = Anim Dynamics - Bone: dyn_tags(ComponentPose=@n4, Alpha=Get ease_in_animdyn() -> (ease_in_animdyn:double), ExternalForce="0,0,0") {
    angularBiasOverride=1
    angularDampingOverride=0.98
    bGravityOverrideInSimSpace=true
    lODThreshold=1
    numSolverIterationsPostUpdate=4
    numSolverIterationsPreUpdate=16
    alphaBoolBlend {blendOption=Linear}
    alphaScaleBiasClamp {bInterpResult=true, interpSpeedIncreasing=0.01, scale=0.3}
    boundBone {boneName=dyn_tags}
    physicsBodyDefinitions [{
      sphereCollisionRadius=0
      boundBone {boneName=dyn_tags}
      boxExtents=<5 2 10>
      constraintSetup {twistAxis=AxisZ, angularLimitsMax=<25 10 0>, angularLimitsMin=<0 -10 0>, angularTarget=<0 0 0>}
      localJointOffset=<0 1 -1>
     }]
    relativeSpaceBone {boneName=dyn_tags}
   }
   n6 = "Constraint - knee_pad_l (PP)"!(ComponentPose=@n5, Alpha=1, ConstraintWeights_0=0.5, ConstraintWeights_1:float=0.5) -> (Pose) {
    alphaBoolBlend {blendOption=Linear}
    boneToModify {boneName=knee_pad_l}
    constraintSetup [{targetBone {boneName=thigh_l}}, {targetBone {boneName=calf_l}}]
    constraintWeights [1,1]
   }
   n7 = RigidBody(ComponentPose=@n6) {
    bClampLinearTranslationLimitToRefPose=true
    bOverrideWorldGravity=true
    overridePhysicsAsset=/Game/Characters/Player/Female/Medium/Bodies/F_Med_Soldier_01/Meshes/F_MED_Soldier_01_Dyn_legs
    simulationSpace=BaseBoneSpace
    alphaBoolBlend {blendOption=Linear}
    baseBoneRef {boneName=spine_01}
    componentAppliedLinearAccClamp=<2000 2000 2000>
    componentLinearAccScale=<0.5 0.5 0.5>
    componentLinearVelScale=<0.5 0.5 0.5>
    overrideWorldGravity {z=-980}
   }
   n8 = "Transform (Modify) Bone - Bone: dyn_r_pocket"(ComponentPose=@n7, Alpha=1, Translation="0,0,0", Rotation="0,10,0", Scale="1,1,1") {
    rotationMode=BMM_Additive
    rotationSpace=BCS_BoneSpace
    alphaBoolBlend {blendOption=Linear}
    boneToModify {boneName=dyn_r_pocket}
    rotation {yaw=10}
   }
   n9 = "Transform (Modify) Bone - Bone: dyn_l_pocket"(ComponentPose=@n8, Alpha=1, Translation="0,0,0", Rotation="0,30,0", Scale="1,1,1") {
    rotationMode=BMM_Additive
    rotationSpace=BCS_BoneSpace
    alphaBoolBlend {blendOption=Linear}
    boneToModify {boneName=dyn_l_pocket}
    rotation {yaw=30}
   }
   Output Pose(Result=Component To Local(ComponentPose=@n9)) {layerGroup=None}
  ]}
 EventGraph {nodes [
   Event OnCharacterCustomizationLoaded()
   > Set ease_in_animdyn(ease_in_animdyn:double=1) -> (Output_Get:double)
   Event BlueprintUpdateAnimation()
   > n1 = Cast To FortPlayerPawn(Object=GetOwningActor())
   > Is Valid(InputObject=@n1)
   > n2 = HasMatchingGameplayTag!(execute:Exec, self:Object Reference=@n1, TagToCheck:Gameplay Tag Structure=(TagName="Item.Gadget.JumpBoots.AnimTracker")) -> (then:Exec, ReturnValue:bool) {is=title K2Node_Message}
   > Branch(Condition=@n2)
   > n3 = Get Dynamic Material Instance(SkeletalMeshComponent=GetPartSkeletalMeshComponent())
   > SetScalarParameterValue(self=@n3, ParameterName=Hide Element 03, Value=1)
   Event BlueprintInitializeAnimation()
   > Set MeshToCopy(MeshToCopy:Skeletal Mesh Component Object Reference=GetOwnerSkeletalMeshComponent()) -> (Output_Get:Skeletal Mesh Component Object Reference)
  ]}
]
variables [
 ease_in_animdyn {type=number}
 MeshToCopy {type=SkeletalMeshComponent}
]
