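# Physics asset for the FuzzyBearIndependence_Star female cosmetic's FaceAcc part, covering dynamic bangs, two buns and their hanging strands.
# Capsules on head, neck, upper spine, clavicles and upper arms form the body volumes.
# Angularly limited four-bone bang chains and bun chains hang from the head; dyn_bun_r_01 alone is rotation-locked to its master.
# Each bun tip carries seven swing-limited strands through tiny MSTR anchor bodies with free-rotation joints.
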PhysicsAsset /BRCosmetics/Characters/Player/Female/Medium/Bodies/F_MED_FuzzyBearIndependence_Star/Meshes/Parts/F_MED_FuzzyBearIndependence_Star_FaceAcc_Physics
bodies {
 clavicle_l [{shapeType=Capsule, length=5.9897, radius=5.3, shapeName=clavicle_r_capsule, center=<3.0309 -3.3938 -0.1525>, rotation=<rot -0.0005 -90 -85>}]
 clavicle_r [{shapeType=Capsule, length=6.4897, radius=5, shapeName=clavicle_r_capsule, center=<-3.0987 2.546 -0.3288>, rotation=<rot -0.0005 -90 95>}]
 dyn_bangs_l_01 [{shapeType=Capsule, length=3.75, radius=1, shapeName=dyn_bangs_l_01_capsule, center=<2.995 0 0>, rotation=<rot 90 0 0>}]
 dyn_bangs_l_02 [{shapeType=Capsule, length=3.75, radius=1.25, shapeName=dyn_bangs_l_02_capsule, center=<2.6469 0 0>, rotation=<rot 90 0 0>}]
 dyn_bangs_l_03 [{shapeType=Capsule, length=3.75, radius=1.5, shapeName=dyn_bangs_l_03_capsule, center=<2.8885 0 0>, rotation=<rot 90 0 0>}]
 dyn_bangs_l_04 [{shapeType=Capsule, length=3.75, radius=1.5, shapeName=dyn_bangs_l_04_capsule, center=<3.6989 0 0>, rotation=<rot 90 0 0>}]
 dyn_bangs_l_MSTR [{shapeType=Capsule, length=0, radius=1, shapeName=dyn_bangs_l_MSTR_capsule, rotation=<rot 90 0 0>}]
 dyn_bangs_r_01 [{shapeType=Capsule, length=3.75, radius=1, shapeName=dyn_bangs_r_01_capsule, center=<2.995 0 0>, rotation=<rot 90 0 0>}]
 dyn_bangs_r_02 [{shapeType=Capsule, length=3.75, radius=1.25, shapeName=dyn_bangs_r_02_capsule, center=<2.6469 0 0>, rotation=<rot 90 0 0>}]
 dyn_bangs_r_03 [{shapeType=Capsule, length=3.75, radius=1.5, shapeName=dyn_bangs_r_03_capsule, center=<2.8885 0 0>, rotation=<rot 90 0 0>}]
 dyn_bangs_r_04 [{shapeType=Capsule, length=3.75, radius=1.5, shapeName=dyn_bangs_r_04_capsule, center=<3.6989 0 0>, rotation=<rot 90 0 0>}]
 dyn_bangs_r_MSTR [{shapeType=Capsule, length=0, radius=1, shapeName=dyn_bangs_r_MSTR_capsule, rotation=<rot 90 0 0>}]
 dyn_bun_l_01 [{shapeType=Capsule, length=2, radius=2, shapeName=dyn_bun_l_01_capsule, rotation=<rot 90 0 0>}]
 dyn_bun_l_02 [{shapeType=Capsule, length=0, radius=1, shapeName=dyn_bun_l_02_capsule, center=<1.8721 0 0>, rotation=<rot 90 0 0>}]
 dyn_bun_l_MSTR [{shapeType=Capsule, length=0, radius=0, shapeName=dyn_bun_l_MSTR_capsule, rotation=<rot 90 0 0>}]
 dyn_bun_r_01 [{shapeType=Capsule, length=2, radius=2, shapeName=dyn_bun_r_01_capsule, rotation=<rot 90 0 0>}]
 dyn_bun_r_02 [{shapeType=Capsule, length=0, radius=1, shapeName=dyn_bun_l_02_capsule, center=<1.8721 0 0>, rotation=<rot 90 0 0>}]
 dyn_bun_r_MSTR [{shapeType=Capsule, length=0, radius=0, shapeName=dyn_bun_r_MSTR_capsule, rotation=<rot 90 0 0>}]
 dyn_strand_l_01 [{
   shapeType=Capsule
   length=5.5058
   radius=1
   shapeName=dyn_strand_l_01_capsule
   center=<3.0618 0 -0.7479>
   rotation=<rot 85.0001 -179.9991 -179.999>
  }]
 dyn_strand_l_01_MSTR [{shapeType=Capsule, length=0, radius=0.1, shapeName=dyn_strand_l_01_MSTR_capsule, rotation=<rot 0 -90 0>}]
 dyn_strand_l_02 [{shapeType=Capsule, length=6.3773, radius=1, shapeName=dyn_strand_l_02_capsule, center=<2.314 0 -0.5396>, rotation=<rot 90 0 0>}]
 dyn_strand_l_02_MSTR [{shapeType=Capsule, length=0, radius=0.1, shapeName=dyn_strand_l_02_MSTR_capsule, rotation=<rot 0 -90 0>}]
 dyn_strand_l_03 [{shapeType=Capsule, length=5.0074, radius=1, shapeName=dyn_strand_l_03_capsule, center=<3.6711 0 -0.6672>, rotation=<rot 90 0 0>}]
 dyn_strand_l_03_MSTR [{shapeType=Capsule, length=0, radius=0.1, shapeName=dyn_strand_l_03_MSTR_capsule, rotation=<rot 0 -90 0>}]
 dyn_strand_l_04 [{shapeType=Capsule, length=6.2787, radius=1, shapeName=dyn_strand_l_04_capsule, center=<3.8537 0 -0.5895>, rotation=<rot 90 0 0>}]
 dyn_strand_l_04_MSTR [{shapeType=Capsule, length=0, radius=0.1, shapeName=dyn_strand_l_04_MSTR_capsule, rotation=<rot 0 -90 0>}]
 dyn_strand_l_05 [{
   shapeType=Capsule
   length=6.3359
   radius=1
   shapeName=dyn_strand_l_05_capsule
   center=<3.8704 0 -0.7834>
   rotation=<rot 79.9999 -179.9989 -179.9988>
  }]
 dyn_strand_l_05_MSTR [{shapeType=Capsule, length=0, radius=0.1, shapeName=dyn_strand_l_05_MSTR_capsule, rotation=<rot 0 -90 0>}]
 dyn_strand_l_06 [{shapeType=Capsule, length=7.7539, radius=1, shapeName=dyn_strand_l_06_capsule, center=<3.6566 0 -0.432>, rotation=<rot 90 0 0>}]
 dyn_strand_l_06_MSTR [{shapeType=Capsule, length=0, radius=0.1, shapeName=dyn_strand_l_06_MSTR_capsule, rotation=<rot 0 -90 0>}]
 dyn_strand_l_07 [{shapeType=Capsule, length=3.25, radius=1, shapeName=dyn_strand_l_07_capsule, center=<3.4059 0 -0.4096>, rotation=<rot 90 0 0>}]
 dyn_strand_l_07_MSTR [{shapeType=Capsule, length=0, radius=0.1, shapeName=dyn_strand_l_07_MSTR_capsule, rotation=<rot 0 -90 0>}]
 dyn_strand_r_01 [{shapeType=Capsule, length=3.9009, radius=1, shapeName=dyn_strand_l_01_capsule, center=<2.9703 0 -0.3753>, rotation=<rot 90 0 180>}]
 dyn_strand_r_01_MSTR [{shapeType=Capsule, length=0, radius=0.1, shapeName=dyn_strand_r_01_MSTR_capsule, rotation=<rot 0 -90 0>}]
 dyn_strand_r_02 [{shapeType=Capsule, length=4.9555, radius=1, shapeName=dyn_strand_l_02_capsule, center=<2.7036 0 -0.7044>, rotation=<rot 90 0 180>}]
 dyn_strand_r_02_MSTR [{shapeType=Capsule, length=0, radius=0.1, shapeName=dyn_strand_r_02_MSTR_capsule, rotation=<rot 0 -90 0>}]
 dyn_strand_r_03 [{
   shapeType=Capsule
   length=4.5672
   radius=1
   shapeName=dyn_strand_l_03_capsule
   center=<4.1478 0 -0.1216>
   rotation=<rot 79.9999 179.9998 -0.0002>
  }]
 dyn_strand_r_03_MSTR [{shapeType=Capsule, length=0, radius=0.1, shapeName=dyn_strand_r_03_MSTR_capsule, rotation=<rot 0 -90 0>}]
 dyn_strand_r_04 [{shapeType=Capsule, length=7.7143, radius=1, shapeName=dyn_strand_r_04_capsule, center=<4.5968 0 -0.7954>, rotation=<rot 90 0 0>}]
 dyn_strand_r_04_MSTR [{shapeType=Capsule, length=0, radius=0.1, shapeName=dyn_strand_r_04_MSTR_capsule, rotation=<rot 0 -90 0>}]
 dyn_strand_r_06 [{
   shapeType=Capsule
   length=7.6875
   radius=1
   shapeName=dyn_strand_r_06_capsule
   center=<3.8313 0 -1.0901>
   rotation=<rot 85 179.9995 179.9995>
  }]
 dyn_strand_r_06_MSTR [{shapeType=Capsule, length=0, radius=0.1, shapeName=dyn_strand_r_06_MSTR_capsule, rotation=<rot 0 -90 0>}]
 dyn_strand_r_07 [{shapeType=Capsule, length=7.4793, radius=1, shapeName=dyn_strand_l_07_capsule, center=<3.213 0 -0.8804>, rotation=<rot 90 0 180>}]
 dyn_strand_r_07_MSTR [{shapeType=Capsule, length=0, radius=0.1, shapeName=dyn_strand_r_07_MSTR_capsule, rotation=<rot 0 -90 0>}]
 dyn_strand_r_08 [{shapeType=Capsule, length=4.9616, radius=1, shapeName=dyn_strand_r_08_capsule, center=<2.6364 0 -0.2595>, rotation=<rot 90 0 0>}]
 dyn_strand_r_08_MSTR [{shapeType=Capsule, length=0, radius=0.1, shapeName=dyn_strand_r_08_MSTR_capsule, rotation=<rot 0 -90 0>}]
 head [
  {shapeType=Capsule, length=5, radius=6.1, shapeName=head_capsule, center=<1.4 2.5969 0>, rotation=<rot 90 0 0>}
  {shapeType=Capsule, length=3.5, radius=1.5, shapeName=head_capsule, center=<0.731 -0.6651 6.1362>, rotation=<rot -75 -0.0011 -20>}
  {
   shapeType=Capsule
   length=3.5
   radius=1.5
   shapeName=head_capsule
   center=<0.8381 -1.5724 -5.9009>
   rotation=<rot -68.5504 147.5556 -165.9911>
  }
 ]
 neck_01 [{shapeType=Capsule, length=0.75, radius=3.25, shapeName=neck_01_capsule, rotation=<rot 90 0 0>}]
 neck_02 [{shapeType=Capsule, length=0.75, radius=3.25, shapeName=neck_02_capsule, rotation=<rot 90 0 0>}]
 spine_05 [{shapeType=Capsule, length=8.3469, radius=9.755, shapeName=spine_05_capsule, center=<0 -2.3933 0>, rotation=<rot 0 -90 0>}]
 upperarm_l [{
   shapeType=Capsule
   length=21.255
   radius=3.755
   shapeName=upperarm_l_capsule
   center=<12.0623 -0.8194 0.3101>
   rotation=<rot -0.002 -89.9982 -90>
  }]
 upperarm_r [{
   shapeType=Capsule
   length=21.255
   radius=3.755
   shapeName=upperarm_l_capsule
   center=<-12.0623 1.1111 -0.3101>
   rotation=<rot -0.002 -89.9982 90>
  }]
}
constraints [
 {
  bone1Name=neck_02
  bone2Name=neck_01
  swing1LimitDegrees=45
  swing1Motion=Free
  swing2LimitDegrees=45
  swing2Motion=Free
  twistLimitDegrees=45
  twistMotion=Free
 }
 {
  bone1Name=head
  bone2Name=neck_02
  swing1LimitDegrees=45
  swing1Motion=Free
  swing2LimitDegrees=45
  swing2Motion=Free
  twistLimitDegrees=45
  twistMotion=Free
 }
 {bone1Name=dyn_bangs_l_01, bone2Name=dyn_bangs_l_MSTR, swing1LimitDegrees=2, swing2LimitDegrees=2, twistLimitDegrees=0}
 {bone1Name=dyn_bangs_l_02, bone2Name=dyn_bangs_l_01, swing1LimitDegrees=10, swing2LimitDegrees=40, twistLimitDegrees=0}
 {bone1Name=dyn_bangs_l_03, bone2Name=dyn_bangs_l_02, swing1LimitDegrees=50, swing2LimitDegrees=30, twistLimitDegrees=5}
 {bone1Name=dyn_bangs_l_04, bone2Name=dyn_bangs_l_03, swing1LimitDegrees=50, swing2LimitDegrees=30, twistLimitDegrees=5}
 {bone1Name=dyn_bangs_r_01, bone2Name=dyn_bangs_r_MSTR, swing1LimitDegrees=2, swing2LimitDegrees=2, twistLimitDegrees=0}
 {bone1Name=dyn_bangs_r_02, bone2Name=dyn_bangs_r_01, swing1LimitDegrees=10, swing2LimitDegrees=40, twistLimitDegrees=0}
 {bone1Name=dyn_bangs_r_03, bone2Name=dyn_bangs_r_02, swing1LimitDegrees=50, swing2LimitDegrees=30, twistLimitDegrees=5}
 {bone1Name=dyn_bangs_r_04, bone2Name=dyn_bangs_r_03, swing1LimitDegrees=50, swing2LimitDegrees=30, twistLimitDegrees=5}
 {bone1Name=dyn_bun_l_MSTR, bone2Name=head, swing1LimitDegrees=10, swing2LimitDegrees=15, twistLimitDegrees=5}
 {bone1Name=dyn_bun_l_01, bone2Name=dyn_bun_l_MSTR, swing1LimitDegrees=20, swing2LimitDegrees=20, twistLimitDegrees=5}
 {bone1Name=dyn_bun_l_02, bone2Name=dyn_bun_l_01, swing1LimitDegrees=5, swing2LimitDegrees=6.5, twistLimitDegrees=1}
 {bone1Name=dyn_strand_l_01, bone2Name=dyn_strand_l_01_MSTR, swing1LimitDegrees=7.5, swing2LimitDegrees=30, twistLimitDegrees=0}
 {bone1Name=dyn_strand_l_02, bone2Name=dyn_strand_l_02_MSTR, swing1LimitDegrees=7.5, swing2LimitDegrees=30, twistLimitDegrees=0}
 {bone1Name=dyn_strand_l_03, bone2Name=dyn_strand_l_03_MSTR, swing1LimitDegrees=25, swing2LimitDegrees=30, twistLimitDegrees=0}
 {bone1Name=dyn_strand_l_04, bone2Name=dyn_strand_l_04_MSTR, swing1LimitDegrees=10, swing2LimitDegrees=40, twistLimitDegrees=0}
 {bone1Name=dyn_strand_l_05, bone2Name=dyn_strand_l_05_MSTR, swing1LimitDegrees=7.5, swing2LimitDegrees=30, twistLimitDegrees=0}
 {bone1Name=dyn_strand_l_06, bone2Name=dyn_strand_l_06_MSTR, swing1LimitDegrees=10, swing2LimitDegrees=30, twistLimitDegrees=0}
 {bone1Name=dyn_strand_l_07, bone2Name=dyn_strand_l_07_MSTR, swing1LimitDegrees=7.5, swing2LimitDegrees=30, twistLimitDegrees=0}
 {bone1Name=dyn_bun_r_MSTR, bone2Name=head, swing1LimitDegrees=20, swing2LimitDegrees=20, twistLimitDegrees=5}
 {
  bone1Name=dyn_bun_r_01
  bone2Name=dyn_bun_r_MSTR
  swing1LimitDegrees=20
  swing1Motion=Locked
  swing2LimitDegrees=20
  swing2Motion=Locked
  twistLimitDegrees=5
  twistMotion=Locked
 }
 {bone1Name=dyn_bun_r_02, bone2Name=dyn_bun_r_01, swing1LimitDegrees=9, swing2LimitDegrees=15, twistLimitDegrees=1}
 {bone1Name=dyn_strand_r_04, bone2Name=dyn_strand_r_04_MSTR, swing1LimitDegrees=7.5, swing2LimitDegrees=30, twistLimitDegrees=0}
 {bone1Name=dyn_strand_r_06, bone2Name=dyn_strand_r_06_MSTR, swing1LimitDegrees=7.5, swing2LimitDegrees=30, twistLimitDegrees=0}
 {bone1Name=dyn_strand_r_08, bone2Name=dyn_strand_r_08_MSTR, swing1LimitDegrees=7.5, swing2LimitDegrees=30, twistLimitDegrees=0}
 {bone1Name=dyn_strand_r_01, bone2Name=dyn_strand_r_01_MSTR, swing1LimitDegrees=20, swing2LimitDegrees=30, twistLimitDegrees=0}
 {bone1Name=dyn_strand_r_07, bone2Name=dyn_strand_r_07_MSTR, swing1LimitDegrees=7.5, swing2LimitDegrees=30, twistLimitDegrees=0}
 {bone1Name=dyn_strand_r_02, bone2Name=dyn_strand_r_02_MSTR, swing1LimitDegrees=7.5, swing2LimitDegrees=30, twistLimitDegrees=0}
 {bone1Name=dyn_strand_r_03, bone2Name=dyn_strand_r_03_MSTR, swing1LimitDegrees=15, swing2LimitDegrees=30, twistLimitDegrees=0}
 {bone1Name=dyn_bangs_r_MSTR, bone2Name=head, swing1LimitDegrees=20, swing2LimitDegrees=20, twistLimitDegrees=5}
 {bone1Name=dyn_bangs_l_MSTR, bone2Name=head, swing1LimitDegrees=20, swing2LimitDegrees=20, twistLimitDegrees=5}
 {
  bone1Name=upperarm_l
  bone2Name=clavicle_l
  swing1LimitDegrees=45
  swing1Motion=Free
  swing2LimitDegrees=45
  swing2Motion=Free
  twistLimitDegrees=45
  twistMotion=Free
 }
 {
  bone1Name=upperarm_r
  bone2Name=clavicle_r
  swing1LimitDegrees=45
  swing1Motion=Free
  swing2LimitDegrees=45
  swing2Motion=Free
  twistLimitDegrees=45
  twistMotion=Free
 }
 {
  bone1Name=neck_01
  bone2Name=spine_05
  swing1LimitDegrees=45
  swing1Motion=Free
  swing2LimitDegrees=45
  swing2Motion=Free
  twistLimitDegrees=45
  twistMotion=Free
 }
 {
  bone1Name=clavicle_r
  bone2Name=spine_05
  swing1LimitDegrees=45
  swing1Motion=Free
  swing2LimitDegrees=45
  swing2Motion=Free
  twistLimitDegrees=45
  twistMotion=Free
 }
 {
  bone1Name=clavicle_l
  bone2Name=spine_05
  swing1LimitDegrees=45
  swing1Motion=Free
  swing2LimitDegrees=45
  swing2Motion=Free
  twistLimitDegrees=45
  twistMotion=Free
 }
 {
  bone1Name=dyn_strand_r_08_MSTR
  bone2Name=dyn_bun_r_02
  swing1LimitDegrees=45
  swing1Motion=Free
  swing2LimitDegrees=45
  swing2Motion=Free
  twistLimitDegrees=45
  twistMotion=Free
 }
 {
  bone1Name=dyn_strand_r_01_MSTR
  bone2Name=dyn_bun_r_02
  swing1LimitDegrees=45
  swing1Motion=Free
  swing2LimitDegrees=45
  swing2Motion=Free
  twistLimitDegrees=45
  twistMotion=Free
 }
 {
  bone1Name=dyn_strand_r_02_MSTR
  bone2Name=dyn_bun_r_02
  swing1LimitDegrees=45
  swing1Motion=Free
  swing2LimitDegrees=45
  swing2Motion=Free
  twistLimitDegrees=45
  twistMotion=Free
 }
 {
  bone1Name=dyn_strand_r_03_MSTR
  bone2Name=dyn_bun_r_02
  swing1LimitDegrees=45
  swing1Motion=Free
  swing2LimitDegrees=45
  swing2Motion=Free
  twistLimitDegrees=45
  twistMotion=Free
 }
 {
  bone1Name=dyn_strand_r_04_MSTR
  bone2Name=dyn_bun_r_02
  swing1LimitDegrees=45
  swing1Motion=Free
  swing2LimitDegrees=45
  swing2Motion=Free
  twistLimitDegrees=45
  twistMotion=Free
 }
 {
  bone1Name=dyn_strand_r_06_MSTR
  bone2Name=dyn_bun_r_02
  swing1LimitDegrees=45
  swing1Motion=Free
  swing2LimitDegrees=45
  swing2Motion=Free
  twistLimitDegrees=45
  twistMotion=Free
 }
 {
  bone1Name=dyn_strand_r_07_MSTR
  bone2Name=dyn_bun_r_02
  swing1LimitDegrees=45
  swing1Motion=Free
  swing2LimitDegrees=45
  swing2Motion=Free
  twistLimitDegrees=45
  twistMotion=Free
 }
 {
  bone1Name=dyn_strand_l_07_MSTR
  bone2Name=dyn_bun_l_02
  swing1LimitDegrees=45
  swing1Motion=Free
  swing2LimitDegrees=45
  swing2Motion=Free
  twistLimitDegrees=45
  twistMotion=Free
 }
 {
  bone1Name=dyn_strand_l_01_MSTR
  bone2Name=dyn_bun_l_02
  swing1LimitDegrees=45
  swing1Motion=Free
  swing2LimitDegrees=45
  swing2Motion=Free
  twistLimitDegrees=45
  twistMotion=Free
 }
 {
  bone1Name=dyn_strand_l_02_MSTR
  bone2Name=dyn_bun_l_02
  swing1LimitDegrees=45
  swing1Motion=Free
  swing2LimitDegrees=45
  swing2Motion=Free
  twistLimitDegrees=45
  twistMotion=Free
 }
 {
  bone1Name=dyn_strand_l_03_MSTR
  bone2Name=dyn_bun_l_02
  swing1LimitDegrees=45
  swing1Motion=Free
  swing2LimitDegrees=45
  swing2Motion=Free
  twistLimitDegrees=45
  twistMotion=Free
 }
 {
  bone1Name=dyn_strand_l_04_MSTR
  bone2Name=dyn_bun_l_02
  swing1LimitDegrees=45
  swing1Motion=Free
  swing2LimitDegrees=45
  swing2Motion=Free
  twistLimitDegrees=45
  twistMotion=Free
 }
 {
  bone1Name=dyn_strand_l_05_MSTR
  bone2Name=dyn_bun_l_02
  swing1LimitDegrees=45
  swing1Motion=Free
  swing2LimitDegrees=45
  swing2Motion=Free
  twistLimitDegrees=45
  twistMotion=Free
 }
 {
  bone1Name=dyn_strand_l_06_MSTR
  bone2Name=dyn_bun_l_02
  swing1LimitDegrees=45
  swing1Motion=Free
  swing2LimitDegrees=45
  swing2Motion=Free
  twistLimitDegrees=45
  twistMotion=Free
 }
]
