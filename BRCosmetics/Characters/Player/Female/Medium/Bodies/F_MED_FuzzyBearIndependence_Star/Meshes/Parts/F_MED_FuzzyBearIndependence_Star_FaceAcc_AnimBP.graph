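# Animation Blueprint for the FuzzyBearIndependence_Star female face-accessory part, driving its hair bun and strand bones.
# It copies the attached parent pose, layering the body mesh's clavicles and upper arms per bone, then applies additive bone offsets for skydiving, emotes and head bend.
# Hair Up comes from the spine_05-to-head transform delta and is scaled down while skydiving.
# A head-relative RigidBody node finishes the pose with state-driven gravity override.
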AnimBlueprint /BRCosmetics/Characters/Player/Female/Medium/Bodies/F_MED_FuzzyBearIndependence_Star/Meshes/Parts/F_MED_FuzzyBearIndependence_Star_FaceAcc_AnimBP
target_skeleton=/Game/Characters/Player/Female/Medium/Bodies/F_MED_FuzzyBearIndependence_Star/Meshes/Parts/F_MED_FuzzyBearIndependence_Star_FaceAcc_Skeleton
graphs [
 AnimGraph {nodes [
   n1 = Get bIsPlayingEmote()
   "Transform (Modify) Bone - Bone: dyn_strand_l_04_MSTR"!(ComponentPose, bAlphaBoolEnabled:bool=@n1, Translation="0,0,0", Rotation="100,0,0", Scale="1,1,1") -> (Pose) {
    alphaInputType=Bool
    rotationMode=BMM_Additive
    rotationSpace=BCS_BoneSpace
    translationMode=BMM_Additive
    alphaBoolBlend {blendOption=Linear}
    boneToModify {boneName=dyn_strand_l_04_MSTR}
    rotation {pitch=100}
   }
   n2 = Get bIsSkyDiving()
   n3 = Get Hair Up() -> (Hair Up:double)
   n4 = "Transform (Modify) Bone - Bone: dyn_bun_l_MSTR"(
    ComponentPose=Local To Component(LocalPose=Use cached pose 'StartPose'())
    Alpha=@n3
    Translation="1,-4,1"
    Rotation="40,0,30"
    Scale="1,1,1"
   ) {
    rotationMode=BMM_Additive
    rotationSpace=BCS_BoneSpace
    translationMode=BMM_Additive
    alphaBoolBlend {blendOption=Linear}
    boneToModify {boneName=dyn_bun_l_MSTR}
    rotation {pitch=40, roll=30}
    translation=<1 -4 1>
   }
   n5 = "Transform (Modify) Bone - Bone: dyn_bun_r_MSTR"(ComponentPose=@n4, Alpha=@n3, Translation="-1,-4,1", Rotation="70,0,10", Scale="1,1,1") {
    rotationMode=BMM_Additive
    rotationSpace=BCS_BoneSpace
    translationMode=BMM_Additive
    alphaBoolBlend {blendOption=Linear}
    boneToModify {boneName=dyn_bun_r_MSTR}
    rotation {pitch=70, roll=10}
    translation=<-1 -4 1>
   }
   n6 = "Transform (Modify) Bone - Bone: dyn_strand_r_03_MSTR"(ComponentPose=@n5, Alpha=@n3, Translation="0,0,0", Rotation="90,0,0", Scale="1,1,1") {
    rotationMode=BMM_Additive
    rotationSpace=BCS_BoneSpace
    translationMode=BMM_Additive
    alphaBoolBlend {blendOption=Linear}
    boneToModify {boneName=dyn_strand_r_03_MSTR}
    rotation {pitch=90}
   }
   n7 = "Transform (Modify) Bone - Bone: dyn_strand_r_01_MSTR"(ComponentPose=@n6, Alpha=1, Translation="0,0,0", Rotation="-20,0,0", Scale="1,1,1") {
    rotationMode=BMM_Additive
    rotationSpace=BCS_BoneSpace
    translationMode=BMM_Additive
    alphaBoolBlend {blendOption=Linear}
    boneToModify {boneName=dyn_strand_r_01_MSTR}
    rotation {pitch=-20}
   }
   n8 = "Transform (Modify) Bone - Bone: dyn_bun_l_MSTR"!(ComponentPose=@n7, bAlphaBoolEnabled:bool=@n2, Translation="0.5,0,0", Rotation="30,0,30", Scale="1,1,1") -> (Pose) {
    alphaInputType=Bool
    rotationMode=BMM_Additive
    rotationSpace=BCS_BoneSpace
    translationMode=BMM_Additive
    alphaBoolBlend {blendOption=Linear}
    boneToModify {boneName=dyn_bun_l_MSTR}
    rotation {pitch=30, roll=30}
    translation {x=0.5}
   }
   n9 = "Transform (Modify) Bone - Bone: dyn_strand_l_04_MSTR"(ComponentPose=@n8, bAlphaBoolEnabled=@n2, Translation="0,0,0", Rotation="30,0,0", Scale="1,1,1") {
    alphaInputType=Bool
    rotationMode=BMM_Additive
    rotationSpace=BCS_BoneSpace
    translationMode=BMM_Additive
    alphaBoolBlend {blendOption=Linear}
    boneToModify {boneName=dyn_strand_l_04_MSTR}
    rotation {pitch=30}
   }
   n10 = "Transform (Modify) Bone - Bone: dyn_bun_l_MSTR"(ComponentPose=@n9, bAlphaBoolEnabled=@n1, Translation="0,0,0.3", Rotation="10,0,0", Scale="1,1,1") {
    alphaInputType=Bool
    rotationMode=BMM_Additive
    rotationSpace=BCS_BoneSpace
    translationMode=BMM_Additive
    alphaBoolBlend {blendOption=Linear}
    boneToModify {boneName=dyn_bun_l_MSTR}
    rotation {pitch=10}
    translation {z=0.3}
   }
   n11 = RigidBody!(ComponentPose=@n10, Alpha, OverrideWorldGravity:Vector=Get GravityOverride(), ExternalForce) -> (Pose) {
    bClampLinearTranslationLimitToRefPose=true
    bOverrideWorldGravity=true
    overridePhysicsAsset=&/F_MED_FuzzyBearIndependence_Star_FaceAcc_Physics
    simulationSpace=BaseBoneSpace
    alphaBoolBlend {blendOption=Linear}
    baseBoneRef {boneName=head}
    componentAppliedLinearAccClamp=<6000 6000 6000>
    componentLinearAccScale=<0.15 0.4 0.5>
    componentLinearVelScale=<0.15 0.4 0.4>
    simSpaceSettings {worldAlpha=0.2}
   }
   Output Pose(Result=Component To Local(ComponentPose=@n11)) {layerGroup=None}
   n12 = Layered blend per bone(
    BasePose=Copy Pose From Mesh() {bCopyCurves=true, bUseAttachedParent=true}
    BlendPoses_0=Copy Pose From Mesh(SourceMeshComponent=Get Body_Mesh())
   ) {layerSetup [{branchFilters [{boneName=clavicle_l}, {boneName=clavicle_r}, {boneName=upperarm_r}, {boneName=upperarm_l}]}]}
   Save cached pose 'StartPose'(Pose=@n12) {cacheName=StartPose}
  ]}
 EventGraph {nodes [
   Event OnCharacterCustomizationLoaded()
   > n1 = Cast To PlayerPawn_Athena(Object=TryGetPawnOwner())
   > Set Body_Mesh(Body_Mesh=GetSkeletalMeshForPartType(self=@n1))
   Event BlueprintInitializeAnimation()
   > Set Skel Mesh(Skel Mesh=GetOwnerSkeletalMeshComponent())
   Event BlueprintUpdateAnimation()
   > n2 = Sequence!(execute) -> (then_0, then_1, then_2:Exec, then_3:Exec, then_4:Exec, then_5:Exec, then_6:Exec)
   n3 = Branch(execute=@n2.then_1, Condition=Get bIsSkyDiving())
   n4 = GetAlphaFromReferenceTransformDelta(SkeletalMeshComponent=Get Skel Mesh(), BoneName=spine_05, BaseName=head, InputRange=(X=0,Y=70), TwistAxis="0,0,1")
   > Set Hair Up(Hair Up:double=float * float(A=@n4, B=0.45) {is=title K2Node_PromotableOperator}) -> (Output_Get:double)
   Set Hair Up(execute=@n3.else, Hair Up=@n4)
   n5 = Gravity Override State Function DEPRECATED(
    execute=@n2.then_0
    CharacterPartAnimInstance=Self-Reference()
    GameTimeInSeconds=GetGameTimeInSeconds()
    GravityOverrideParams=Get Gravity Override Params()
   )
   > Set GravityOverride(GravityOverride=@n5)
  ]}
]
variables [
 Body_Mesh {type=SkeletalMeshComponent}
 Gravity Override Params {
  type=GravityOverrideParamsStruct
  default {skydive_FloatingDown {gravityOverride=<0 -700 2000>}, skydive_GoingUp {gravityOverride=<0 -300 2000>}}
 }
 GravityOverride {type=Vector}
 Hair Up {type=number}
 Is Wearing Cape {type=boolean}
 IsSkydive {type=boolean}
 isWearingBackbling {type=boolean}
 Linear Scale Cape {type=Vector}
 Skel Mesh {type=SkeletalMeshComponent}
]
